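# Physics asset for the F_MED_Sunrise_Castle female medium body cosmetic, covering its dynamic hoodie bones.
# Single capsules sit on three hoodie bones, with larger capsules on neck_01 and spine_05 forming the upper-body shapes around them.
# Both hoodie base bones attach to spine_05 with tight swing and twist limits.
# The hoodie_dyn_Bt bone hangs from BaseB, and the neck from the spine, with free angular motion.
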PhysicsAsset /BRCosmetics/Characters/Player/Female/Medium/Bodies/F_MED_Sunrise_Castle/Meshes/F_MED_Sunrise_Castle_Physics
bodies {
 hoodie_dyn_BaseA [{shapeType=Capsule, length=1.505, radius=1.505, shapeName=hoodie_dyn_BaseA_capsule, center=<-1.2738 0 1.7067>, rotation=<rot 90 180 -89.9999>}]
 hoodie_dyn_BaseB [{shapeType=Capsule, length=1.505, radius=1.505, shapeName=hoodie_dyn_BaseB_capsule, center=<-1.2738 0 1.7067>, rotation=<rot 90 180 -90.0004>}]
 hoodie_dyn_Bt [{
   shapeType=Capsule
   length=4.0007
   radius=2.1974
   shapeName=hoodie_dyn_Bt_capsule
   center=<0.0711 0.5048 0.0817>
   rotation=<rot -0.0945 3.2523 -85.8388>
  }]
 neck_01 [{shapeType=Capsule, length=14, radius=6.005, shapeName=neck_01_capsule, rotation=<rot 90 0 0>}]
 spine_05 [{
   shapeType=Capsule
   length=8.0072
   radius=11.5
   shapeName=spine_05_capsule
   center=<0.0263 0.9374 -0.2447>
   rotation=<rot 0 12.4742 -171.4717>
  }]
}
constraints [
 {bone1Name=hoodie_dyn_BaseA, bone2Name=spine_05, swing1LimitDegrees=20, swing2LimitDegrees=20, twistLimitDegrees=10}
 {bone1Name=hoodie_dyn_BaseB, bone2Name=spine_05, swing1LimitDegrees=20, swing2LimitDegrees=20, twistLimitDegrees=10}
 {
  bone1Name=hoodie_dyn_Bt
  bone2Name=hoodie_dyn_BaseB
  swing1LimitDegrees=45
  swing1Motion=Free
  swing2LimitDegrees=45
  swing2Motion=Free
  twistLimitDegrees=45
  twistMotion=Free
 }
 {
  bone1Name=neck_01
  bone2Name=spine_05
  swing1LimitDegrees=45
  swing1Motion=Free
  swing2LimitDegrees=45
  swing2Motion=Free
  twistLimitDegrees=45
  twistMotion=Free
 }
]
