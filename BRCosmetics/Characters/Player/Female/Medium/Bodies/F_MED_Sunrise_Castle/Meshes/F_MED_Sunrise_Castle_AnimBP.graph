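# Post-processes animation for the F_MED_Sunrise_Castle female medium body cosmetic, starting from the pose copied from its attached parent mesh.
# Two Control Rigs run in series to deform that pose, first twists and then sculpters, the latter exposing knee corrective inputs.
# A RigidBody node then simulates the physics asset in root-bone space at partial alpha with overridden gravity.
# The event graph holds no connected logic.
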AnimBlueprint /BRCosmetics/Characters/Player/Female/Medium/Bodies/F_MED_Sunrise_Castle/Meshes/F_MED_Sunrise_Castle_AnimBP
target_skeleton=&/F_MED_Sunrise_Castle_Skeleton
graphs [
 AnimGraph {nodes [
   n1 = Control Rig(Source=Copy Pose From Mesh() {bUseAttachedParent=true}) {
    alphaBoolBlend {blendOption=Linear}
    controlRigAssetReference {blueprintRigClass=&/F_MED_Sunrise_Castle_Deform_CtrlRig_twists_2.F_MED_Sunrise_Castle_Deform_CtrlRig_twists_2_C}
   }
   n2 = Control Rig!(
    Source=@n1
    Alpha
    KneeRot_fr_l_114deg:Vector
    KneeTrans_fr_l_114deg:Vector="0.258,-3.591,0.009"
    KneeScale_fr_l_114deg:Vector="1,1,1"
    KneeScale_fr_l_130deg:Vector="1,1,1"
    KneeRot_fr_l_130deg:Vector="0,0,-51.2"
    KneeTrans_fr_l_130deg:Vector="0.258,-3.591,0.009"
   ) -> (Pose) {
    alphaBoolBlend {blendOption=Linear}
    controlRigAssetReference {blueprintRigClass=&/F_MED_Sunrise_Castle_Deform_CtrlRig_sculpters_2.F_MED_Sunrise_Castle_Deform_CtrlRig_sculpters_2_C}
   }
   n3 = RigidBody(ComponentPose=Local To Component(LocalPose=@n2), Alpha=0.8) {
    bOverrideWorldGravity=true
    overridePhysicsAsset=&/F_MED_Sunrise_Castle_Physics
    simulationSpace=BaseBoneSpace
    alphaBoolBlend {blendOption=Linear}
    baseBoneRef {boneName=root}
    overrideWorldGravity {z=-980}
   }
   Output Pose(Result=Component To Local(ComponentPose=@n3)) {layerGroup=None}
  ]}
 EventGraph {nodes [
   GetAlphaFromReferenceTransformDelta(BoneName=None, BaseName=None, DeltaComponentType=QuaternionTwist, InputRange=(X=0,Y=1), TwistAxis="1,0,0")
   Event BlueprintUpdateAnimation()
  ]}
]
variables []
