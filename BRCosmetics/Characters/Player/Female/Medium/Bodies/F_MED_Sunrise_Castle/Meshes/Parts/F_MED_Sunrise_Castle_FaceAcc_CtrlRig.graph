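# Control Rig for the face accessory part of the F_MED_Sunrise_Castle female player body that shifts the hair bun base bone as the head rotates.
# It reads the head's rotation relative to spine_05, converts it to Euler angles and remaps the Z angle through a clamped range.
# The result drives the Z translation of hairbun_dyn_base in a Modify Transforms step, a procedural corrective with no controls.
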ControlRigBlueprint /BRCosmetics/Characters/Player/Female/Medium/Bodies/F_MED_Sunrise_Castle/Meshes/Parts/F_MED_Sunrise_Castle_FaceAcc_CtrlRig
graphs [
 RigVMFunctionLibrary {links [], nodes []}
 RigVMModel {nodes [
   n1 = &RigUnit_BeginExecution()
   n2 = &RigUnit_GetRelativeTransformForItem(Child=(Type=Bone,Name="head"), Parent=(Type=Bone,Name="spine_05"))
   &RigUnit_ModifyTransforms(
    ItemToModify=((Item=(Type=Bone,Name="hairbun_dyn_base"),Transform=(Rotation=(X=0,Y=0,Z=0,W=1),Translation=(X=0,Y=0,Z=0),Scale3D=(X=1,Y=1,Z=1))))
    Weight=1
    ItemToModify.0.Transform.Translation.Z=&RigVMFunction_MathFloatRemap(
     SourceMaximum=-90
     SourceMinimum=-20
     TargetMaximum=10
     bClamp=true
     Value=&RigVMFunction_MathQuaternionToEuler(RotationOrder=ZYX, Value=@n2.RelativeTransform.Rotation).Result.Z
    ).Result
    ExecutePin=@n1.ExecutePin
   )
  ]}
]
hierarchy [
 root {loc=<0 0 0>, rot=<0 0 0 1>}
 pelvis {parent=root, loc=<0 -0.5301 98.0571>, rot=<0.0127 0.707 -0.0127 -0.707>}
 spine_01 {parent=pelvis, loc=<6.1208 -0.9328 0>, rot=<0 0 0.034 -0.9994>}
 spine_02 {parent=spine_01, loc=<6.5996 -0.2095 0>, rot=<0 0 0.0238 -0.9997>}
 spine_03 {parent=spine_02, loc=<6.4914 -0.1237 0>, rot=<0 0 0.0149 -0.9999>}
 spine_04 {parent=spine_03, loc=<6.9598 0.0794 0>, rot=<0 0 -0.0132 -0.9999>}
 spine_05 {parent=spine_04, loc=<6.9987 0 0>, rot=<0 0 -0.0015 -1>}
 clavicle_l {parent=spine_05, loc=<5.4831 -0.1053 -3.2946>, rot=<0.0465 -0.7289 -0.009 -0.6829>}
 clavicle_r {parent=spine_05, loc=<5.4831 -0.1053 3.2946>, rot=<0.7289 0.0465 -0.6829 0.009>}
 neck_01 {parent=spine_05, loc=<11.3138 0.1529 0>, rot=<0 0 -0.2003 -0.9797>}
 neck_02 {parent=neck_01, loc=<4.5627 0 0>, rot=<0 0 0.0504 -0.9987>}
 head {parent=neck_02, loc=<5.3914 0 0>, rot=<0 0 0.115 -0.9934>}
 faceAttach {parent=head, loc=<-0.1748 -0.0493 0>, rot=<0.0082 -0.7071 -0.0082 -0.7071>, scale=<0.89 0.89 0.89>}
 C_brow_mid {parent=faceAttach, loc=<-0.0019 9.1997 5.3811>, rot=<0 0 0 -1>}
 L_brow_mid {parent=faceAttach, loc=<3.2477 8.9096 5.9841>, rot=<-0.0001 -0.0042 0.2553 -0.9669>}
 L_brow_outer {parent=faceAttach, loc=<5.7983 6.59 5.234>, rot=<-0.0286 -0.04 0.4491 -0.8921>}
 L_cheek_inner {parent=faceAttach, loc=<4.1627 7.5994 -0.1346>, rot=<-0.0706 -0.2697 0.3184 -0.906>}
 C_lip_upper_mid {parent=faceAttach, loc=<0.0024 9.9844 -2.4143>, rot=<0.2225 0.0013 -0.0005 -0.9749>}
 L_lip_upper_outer {parent=faceAttach, loc=<1.5644 9.4494 -2.4491>, rot=<0.0359 -0.0728 0.2627 -0.9615>}
 L_lip_corner {parent=faceAttach, loc=<2.6143 8.3858 -2.9118>, rot=<0.0029 0.008 -0.0146 -0.9999>}
 C_nose_bridge {parent=faceAttach, loc=<-0.0198 10.9089 1.7137>, rot=<-0.2218 0 0 -0.9751>}
 R_brow_mid {parent=faceAttach, loc=<-3.2428 8.901 5.9841>, rot=<-0.0001 0.0042 -0.2553 -0.9669>}
 R_brow_outer {parent=faceAttach, loc=<-5.7792 6.5767 5.235>, rot=<-0.0287 0.04 -0.4491 -0.8921>}
 R_cheek_inner {parent=faceAttach, loc=<-4.1627 7.5994 -0.1346>, rot=<-0.0706 0.2697 -0.3184 -0.906>}
 R_lip_corner {parent=faceAttach, loc=<-2.6143 8.3858 -2.9118>, rot=<0.0029 -0.008 0.0146 -0.9999>}
 C_jaw {parent=faceAttach, loc=<-0.0151 -0.1812 -1.6057>, rot=<0 0 0 1>}
 L_lip_lower_outer {parent=C_jaw, loc=<1.5534 9.4086 -1.6001>, rot=<0.013 0.0445 0.3056 -0.951>}
 C_lip_lower_mid {parent=C_jaw, loc=<0.0251 9.8762 -1.665>, rot=<0 0 0 -1>}
 R_lip_lower_outer {parent=C_jaw, loc=<-1.4665 9.3228 -1.6>, rot=<0.013 -0.0445 -0.3057 -0.951>}
 teeth_lower {parent=C_jaw, loc=<0.0151 6.8354 -1.9816>, rot=<0 0 0 1>}
 tongue {parent=C_jaw, loc=<0.0158 3.4219 -0.4018>, rot=<-0.1568 0 0 -0.9876>}
 R_eye {parent=faceAttach, loc=<-3.3012 7.343 4.1631>, rot=<0 0 0.9998 -0.0187>}
 R_eye_lid_upper_mid {parent=R_eye, loc=<-0.0396 -1.1714 0.2631>, rot=<0.0523 -0.0009 0.0167 -0.9985>}
 R_eye_lid_lower_mid {parent=R_eye, loc=<-0.0085 -0.9425 -0.7051>, rot=<0.0329 -0.0005 0.0167 -0.9993>}
 L_eye {parent=faceAttach, loc=<3.2705 7.3431 4.1595>, rot=<0 0 0.9998 0.0187>}
 L_eye_lid_upper_mid {parent=L_eye, loc=<0.0251 -1.1623 0.2388>, rot=<0.0523 0 0 -0.9986>}
 L_eye_lid_lower_mid {parent=L_eye, loc=<0.0223 -0.923 -0.7488>, rot=<0.0329 0 0 -0.9995>}
 teeth_upper {parent=faceAttach, loc=<0 7.3065 -1.2818>, rot=<0 0 0 1>}
 R_lip_upper_outer {parent=faceAttach, loc=<-1.5644 9.4494 -2.4491>, rot=<0.0359 0.0728 -0.2627 -0.9615>}
 hair_dyn_rt_a_base {parent=head, loc=<8.5389 8.7985 4.5735>, rot=<0.0471 -0.0043 0.994 -0.0983>}
 hair_dyn_rt_a_01 {parent=hair_dyn_rt_a_base, loc=<1.2699 0 0>, rot=<-0.0062 -0.1542 0.1463 -0.9771>}
 hair_dyn_rt_a_02 {parent=hair_dyn_rt_a_01, loc=<2.2289 0 0>, rot=<0.0062 0.2487 -0.1289 -0.9599>}
 hair_dyn_rt_bc_base {parent=head, loc=<2.0138 4.0985 6.855>, rot=<-0.0672 0.0211 0.9491 -0.3072>}
 hair_dyn_rt_c_01 {parent=hair_dyn_rt_bc_base, loc=<0.7159 0 0>, rot=<0.132 0.2406 0.1113 -0.9552>}
 hair_dyn_rt_c_02 {parent=hair_dyn_rt_c_01, loc=<1.6531 0 0>, rot=<0.068 0.1094 -0.1684 -0.9773>}
 hair_dyn_rt_b_01 {parent=hair_dyn_rt_bc_base, loc=<0.4204 -0.6483 -0.1913>, rot=<0.0034 0.0103 0.2954 -0.9553>}
 hair_dyn_rt_b_02 {parent=hair_dyn_rt_b_01, loc=<1.9171 0 0>, rot=<0.0169 0.2162 -0.0586 -0.9744>}
 hair_dyn_lf_a_base {parent=head, loc=<8.08 8.7063 -4.2134>, rot=<-0.1645 0.0375 0.9592 -0.2269>}
 hair_dyn_lf_a_01 {parent=hair_dyn_lf_a_base, loc=<1.5152 0 0>, rot=<-0.0358 -0.0729 0 -0.9967>}
 hair_dyn_lf_a_02 {parent=hair_dyn_lf_a_01, loc=<1.7024 0 0>, rot=<0.0321 0.1082 0.1658 -0.9797>}
 hair_dyn_lf_a_03 {parent=hair_dyn_lf_a_02, loc=<2.1693 0 0>, rot=<-0.0137 -0.0326 -0.2723 -0.9616>}
 hair_dyn_lf_b_base {parent=head, loc=<5.3913 5.183 -6.4982>, rot=<0.0115 -0.0006 0.9979 -0.0632>}
 hair_dyn_lf_b_01 {parent=hair_dyn_lf_b_base, loc=<1.6751 0 0>, rot=<-0.0001 -0.0006 -0.0538 -0.9986>}
 hair_dyn_lf_b_02 {parent=hair_dyn_lf_b_01, loc=<2.4867 0 0>, rot=<-0.1023 -0.2241 -0.1947 -0.9494>}
 hair_dyn_rt_d_base {parent=head, loc=<1.3494 0.9684 7.1533>, rot=<0.0286 -0.009 0.9512 -0.307>}
 hair_dyn_rt_d_01 {parent=hair_dyn_rt_d_base, loc=<2.1442 0 0>, rot=<-0.0145 -0.0666 0.3901 -0.9183>}
 hairbun_dyn_base {parent=head, loc=<2.5186 -3.9502 -0.2358>, rot=<0.3361 -0.6221 0.6221 -0.3361>}
 hairbun_dyn_01 {parent=hairbun_dyn_base, loc=<7.325 0 0>, rot=<0 0.0152 0 -0.9999>}
]
