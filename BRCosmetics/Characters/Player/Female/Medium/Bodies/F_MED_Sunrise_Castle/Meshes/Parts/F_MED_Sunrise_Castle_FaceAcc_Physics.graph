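# Physics asset for the F_MED_Sunrise_Castle face accessory mesh, providing bodies and joints for dynamic hair strands on a female player cosmetic.
# Small capsules form left and right hair chains, with five base bodies constrained to a head body built from two large capsules.
# Base-to-head joints swing and twist freely, while links down each strand have narrow swing limits and zero twist.
# Right strands B and C branch from one shared base.
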PhysicsAsset /BRCosmetics/Characters/Player/Female/Medium/Bodies/F_MED_Sunrise_Castle/Meshes/Parts/F_MED_Sunrise_Castle_FaceAcc_Physics
bodies {
 hair_dyn_lf_a_01 [{
   shapeType=Capsule
   length=0.8433
   radius=0.4
   shapeName=hair_dyn_lf_a_01_capsule
   center=<0.5806 0.0249 0.0434>
   rotation=<rot 0 86.0411 -88.4979>
  }]
 hair_dyn_lf_a_02 [{
   shapeType=Capsule
   length=0.505
   radius=0.505
   shapeName=hair_dyn_lf_a_02_capsule
   center=<1.0104 -0.0892 -0.0916>
   rotation=<rot 0 -95.1108 -86.2691>
  }]
 hair_dyn_lf_a_03 [{
   shapeType=Capsule
   length=0.505
   radius=0.505
   shapeName=hair_dyn_lf_a_03_capsule
   center=<0.913 -0.1109 0.016>
   rotation=<rot 0 -93.0643 -84.5334>
  }]
 hair_dyn_lf_a_base [{
   shapeType=Capsule
   length=0.6147
   radius=0.5157
   shapeName=hair_dyn_lf_a_base_capsule
   center=<0.7221 -0.0866 0.0033>
   rotation=<rot 9.5107 -118.2083 -90>
  }]
 hair_dyn_lf_b_01 [{
   shapeType=Capsule
   length=2.505
   radius=0.505
   shapeName=hair_dyn_lf_b_01_capsule
   center=<1.1986 0.0518 -0.2539>
   rotation=<rot 0 94.0118 -91.2359>
  }]
 hair_dyn_lf_b_02 [{
   shapeType=Capsule
   length=1.255
   radius=0.505
   shapeName=hair_dyn_lf_b_02_capsule
   center=<1.1997 -0.0558 -0.0343>
   rotation=<rot 0 93.9227 -89.515>
  }]
 hair_dyn_lf_b_base [{shapeType=Capsule, length=0.505, radius=0.505, shapeName=hair_dyn_lf_b_base_capsule, rotation=<rot 0 -90 0>}]
 hair_dyn_rt_a_01 [{
   shapeType=Capsule
   length=1.1914
   radius=0.3
   shapeName=hair_dyn_rt_a_01_capsule
   center=<0.8502 -0.0514 0.0964>
   rotation=<rot 0 -93.3569 -81.5592>
  }]
 hair_dyn_rt_a_02 [{
   shapeType=Capsule
   length=0.505
   radius=0.3
   shapeName=hair_dyn_rt_a_02_capsule
   center=<0.9705 0.0431 0.0813>
   rotation=<rot 0 106.692 -74.2567>
  }]
 hair_dyn_rt_a_base [{shapeType=Capsule, length=0.505, radius=0.505, shapeName=hair_dyn_rt_a_base_capsule, rotation=<rot 0 -90 0>}]
 hair_dyn_rt_b_01 [{
   shapeType=Capsule
   length=1.41
   radius=0.5
   shapeName=hair_dyn_rt_b_01_capsule
   center=<0.6797 -0.0334 0.1102>
   rotation=<rot 0 -90.5567 -88.4045>
  }]
 hair_dyn_rt_b_02 [{
   shapeType=Capsule
   length=0.755
   radius=0.505
   shapeName=hair_dyn_rt_b_02_capsule
   center=<0.8934 -0.0165 -0.0702>
   rotation=<rot 0 89.6384 -88.363>
  }]
 hair_dyn_rt_bc_base [{
   shapeType=Capsule
   length=0.505
   radius=0.505
   shapeName=hair_dyn_rt_bc_base_capsule
   center=<-0.1589 -0.3836 -0.3634>
   rotation=<rot 0 -21.2452 -78.1>
  }]
 hair_dyn_rt_c_01 [{
   shapeType=Capsule
   length=1.005
   radius=0.7
   shapeName=hair_dyn_rt_c_01_capsule
   center=<0.6215 0.0345 -0.0067>
   rotation=<rot 0 97.7542 -88.4082>
  }]
 hair_dyn_rt_c_02 [{
   shapeType=Capsule
   length=0.505
   radius=0.505
   shapeName=hair_dyn_rt_c_02_capsule
   center=<0.4762 0.052 0.0264>
   rotation=<rot 0 94.1917 -86.8442>
  }]
 hair_dyn_rt_d_01 [{
   shapeType=Capsule
   length=0.505
   radius=0.505
   shapeName=hair_dyn_rt_d_01_capsule
   center=<0.4928 0.0271 -0.0058>
   rotation=<rot 0 87.3218 -88.5761>
  }]
 hair_dyn_rt_d_base [{shapeType=Capsule, length=0.505, radius=0.505, shapeName=hair_dyn_rt_d_base_capsule, rotation=<rot 0 -90 0>}]
 head [
  {shapeType=Capsule, length=1.5342, radius=6.2981, center=<4.423 2.8534 0.0295>, rotation=<rot 0 167.3372 -87.7763>}
  {shapeType=Capsule, length=7.5, radius=6.5, shapeName=head_capsule, center=<4.0315 2.0101 -0.0832>, rotation=<rot 0 74.8665 89.9998>}
 ]
}
constraints [
 {bone1Name=hair_dyn_rt_d_01, bone2Name=hair_dyn_rt_d_base, swing1LimitDegrees=15, swing2LimitDegrees=15, twistLimitDegrees=0}
 {bone1Name=hair_dyn_lf_b_02, bone2Name=hair_dyn_lf_b_01, swing1LimitDegrees=15, swing2LimitDegrees=15, twistLimitDegrees=0}
 {bone1Name=hair_dyn_lf_b_01, bone2Name=hair_dyn_lf_b_base, swing1LimitDegrees=15, swing2LimitDegrees=15, twistLimitDegrees=0}
 {bone1Name=hair_dyn_lf_a_03, bone2Name=hair_dyn_lf_a_02, swing1LimitDegrees=15, swing2LimitDegrees=15, twistLimitDegrees=0}
 {bone1Name=hair_dyn_lf_a_02, bone2Name=hair_dyn_lf_a_01, swing1LimitDegrees=15, swing2LimitDegrees=15, twistLimitDegrees=0}
 {bone1Name=hair_dyn_lf_a_01, bone2Name=hair_dyn_lf_a_base, swing1LimitDegrees=15, swing2LimitDegrees=15, twistLimitDegrees=0}
 {bone1Name=hair_dyn_rt_b_02, bone2Name=hair_dyn_rt_b_01, swing1LimitDegrees=15, swing2LimitDegrees=15, twistLimitDegrees=0}
 {bone1Name=hair_dyn_rt_b_01, bone2Name=hair_dyn_rt_bc_base, swing1LimitDegrees=15, swing2LimitDegrees=15, twistLimitDegrees=0}
 {bone1Name=hair_dyn_rt_c_02, bone2Name=hair_dyn_rt_c_01, swing1LimitDegrees=15, swing2LimitDegrees=15, twistLimitDegrees=0}
 {bone1Name=hair_dyn_rt_c_01, bone2Name=hair_dyn_rt_bc_base, swing1LimitDegrees=15, swing2LimitDegrees=15, twistLimitDegrees=0}
 {bone1Name=hair_dyn_rt_a_02, bone2Name=hair_dyn_rt_a_01, swing1LimitDegrees=15, swing2LimitDegrees=15, twistLimitDegrees=0}
 {bone1Name=hair_dyn_rt_a_01, bone2Name=hair_dyn_rt_a_base, swing1LimitDegrees=15, swing2LimitDegrees=15, twistLimitDegrees=0}
 {
  bone1Name=hair_dyn_rt_a_base
  bone2Name=head
  swing1LimitDegrees=45
  swing1Motion=Free
  swing2LimitDegrees=45
  swing2Motion=Free
  twistLimitDegrees=45
  twistMotion=Free
 }
 {
  bone1Name=hair_dyn_rt_bc_base
  bone2Name=head
  swing1LimitDegrees=45
  swing1Motion=Free
  swing2LimitDegrees=45
  swing2Motion=Free
  twistLimitDegrees=45
  twistMotion=Free
 }
 {
  bone1Name=hair_dyn_lf_a_base
  bone2Name=head
  swing1LimitDegrees=45
  swing1Motion=Free
  swing2LimitDegrees=45
  swing2Motion=Free
  twistLimitDegrees=45
  twistMotion=Free
 }
 {
  bone1Name=hair_dyn_lf_b_base
  bone2Name=head
  swing1LimitDegrees=45
  swing1Motion=Free
  swing2LimitDegrees=45
  swing2Motion=Free
  twistLimitDegrees=45
  twistMotion=Free
 }
 {
  bone1Name=hair_dyn_rt_d_base
  bone2Name=head
  swing1LimitDegrees=45
  swing1Motion=Free
  swing2LimitDegrees=45
  swing2Motion=Free
  twistLimitDegrees=45
  twistMotion=Free
 }
]
